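; PCG sample graph that turns static mesh sockets into points and places them at tagged actors in the world.
; Two MeshSocketsToPoints Blueprint elements read the same socketed cube mesh, each with a different tag.
; Get Actor Data gathers every MyActor-tagged actor as a single point, and the results are merged.
; Two Copy Points nodes stamp each socket point set onto those actor points, feeding one Debug node for visualization.
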(graph MeshSocketsToPointsGraph
  (bind BlueprintNode_0 (Execute%20Blueprint :BlueprintElementInstance (class:/PCG/SampleContent/MeshSockets/Elements/MeshSocketsToPoints.MeshSocketsToPoints_C :StaticMesh /PCG/SampleContent/MeshSockets/Meshes/1M_CubeWithSocket.1M_CubeWithSocket :Tag "MyTag") :BlueprintElementType /PCG/SampleContent/MeshSockets/Elements/MeshSocketsToPoints.MeshSocketsToPoints_C))
  (bind BlueprintNode_4 (Execute%20Blueprint :BlueprintElementInstance (class:/PCG/SampleContent/MeshSockets/Elements/MeshSocketsToPoints.MeshSocketsToPoints_C :StaticMesh /PCG/SampleContent/MeshSockets/Meshes/1M_CubeWithSocket.1M_CubeWithSocket :Tag "MySecondTag") :BlueprintElementType /PCG/SampleContent/MeshSockets/Elements/MeshSocketsToPoints.MeshSocketsToPoints_C))
  (bind GetActorData_4 (Get%20Actor%20Data :ActorSelector (struct :ActorFilter AllWorldActors :ActorSelectionTag MyActor :ActorReferenceSelector @LastCreated :bSelectMultiple true) :bComponentsMustOverlapSelf false :bTrackActorsOnlyWithinBounds false :Mode GetSinglePoint))
  (bind Merge_5 (Merge%20Points))
  (bind CopyPoints_1 (Copy%20Points))
  (bind CopyPoints_3 (Copy%20Points))
  (bind Debug_0 (Debug))
  (connect BlueprintNode_4 Out CopyPoints_1 Source)
  (connect Merge_5 Out CopyPoints_1 Target)
  (connect BlueprintNode_0 Out CopyPoints_3 Source)
  (connect Merge_5 Out CopyPoints_3 Target)
  (connect CopyPoints_1 Out Debug_0 In)
  (connect CopyPoints_3 Out Debug_0 In)
  (connect GetActorData_4 Out Merge_5 In))
